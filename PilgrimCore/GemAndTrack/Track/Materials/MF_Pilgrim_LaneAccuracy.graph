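# Material function for Pilgrim track lanes that builds the hit accuracy feedback color and mask.
# The HitFeedbackData red channel blends Good to Perfect color and gradient, while blue picks between the Early and Late gradients for Good hits, which get an intensity boost.
# The chosen gradient, LaneMask and the alpha channel multiply into the LaneMask output, and also tint the Color output.
# The green channel is unused.
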MaterialFunction /PilgrimCore/GemAndTrack/Track/Materials/MF_Pilgrim_LaneAccuracy
graph {
 properties {bExposeToLibrary=true, libraryCategoriesText [Pilgrim]}
 nodes [
  n1 = FunctionInput() {inputName=HitFeedbackData, inputType=FunctionInput_Vector4}
  n2 = ComponentMask(None=@n1) {r=true}
  n3 = FunctionInput() {inputName=EarlyAccuracyGradient, inputType=FunctionInput_Scalar, sortPriority=5}
  n4 = FunctionInput() {inputName=LateAccuracyGradient, inputType=FunctionInput_Scalar, sortPriority=6}
  n5 = ComponentMask(None=@n1) {b=true}
  n6 = LinearInterpolate(
   A=Multiply(A=LinearInterpolate(A=@n3, B=@n4, Alpha=@n5)) {constB=1.5, desc=Boost Goods Intensity}
   B=FunctionInput() {inputName=AccuracyGradient, inputType=FunctionInput_Scalar, sortPriority=4}
   Alpha=@n2
  )
  n7 = Multiply(
   A=LinearInterpolate(
    A=FunctionInput() {inputName=GoodColor, sortPriority=2, previewValue=<0 0 1 1>}
    B=FunctionInput() {inputName=ParfectColor, sortPriority=1, previewValue=<1 1 0 1>}
    Alpha=@n2
   )
   B=@n6
  )
  n8 = FunctionInput() {inputName=LaneMask, inputType=FunctionInput_Scalar, sortPriority=7}
  n9 = ComponentMask(None=@n1) {a=true}
  FunctionOutput(None=Multiply(A=Multiply(A=@n7, B=@n8), B=@n9)) {outputName=Color}
  FunctionOutput(None=Multiply(A=@n9, B=Multiply(A=@n6, B=@n8))) {outputName=LaneMask, sortPriority=5}
 ]
}
