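# Material function for Pilgrim track beat markers that converts distance from the now bar into a fade Alpha.
# A SmoothStep runs from zero at the now bar to one at the optional FadeDistance, then saturates.
# FadeDistance defaults to a negative value, so by default only negative distances raise Alpha.
# The FadeDistance description calls that point fully invisible, although Alpha outputs one there.
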MaterialFunction /PilgrimCore/GemAndTrack/Track/Materials/MF_GetPilgrimBeatMarkerNowBarToAlpha
graph {nodes [
  n1 = Constant()
  n2 = FunctionInput() {
   bUsePreviewValueAsDefault=true
   desc="At what distance from the now bar should we be completely invisible. (Optional)"
   inputName=FadeDistance
   inputType=FunctionInput_Scalar
   previewValue=<-12.5 0 0 1>
  }
  n3 = FunctionInput() {inputName=DistanceFromNowBar, inputType=FunctionInput_Scalar}
  FunctionOutput(None=Saturate(None=SmoothStep(Min=@n1, Max=@n2, Value=@n3))) {outputName=Alpha}
 ]}
