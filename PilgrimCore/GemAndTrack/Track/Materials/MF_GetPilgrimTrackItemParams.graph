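# Material function that unpacks a Vector4 of per-item data for Pilgrim track items, such as notes.
# R splits into track and lane indices at five lanes per track; B and A pass through as last-frame and current now-bar distances.
# A Custom node reinterprets G as int bitflags for pro colors, overdrive and world-position offsets.
# It uses logical `&&`, so higher bits also trigger lower flags; two debug constants are unconnected.
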MaterialFunction /PilgrimCore/GemAndTrack/Track/Materials/MF_GetPilgrimTrackItemParams
graph {nodes [
  Constant() {desc=Force Pro Colors Debug, r=1}
  Constant() {desc=Force All Notes Overdrive Debug, r=1}
  n1 = FunctionInput() {
   desc="R: TrackAndLaneIndex\r\nG: BitFlags\r\nB: Last Frames Distance From Now Bar\r\nA: Current Distance From Now Bar"
   inputType=FunctionInput_Vector4
  }
  MaterialExpressionMaterialFunctionCall_3 = MaterialFunctionCall!(Float4=@n1) {materialFunction=/Game/Global/MaterialFunctions/MF_BreakComponents}
  FunctionOutput(None=@MaterialExpressionMaterialFunctionCall_3.B) {outputName=Last Frame Distance From Now Bar, sortPriority=6}
  n2 = Constant() {desc=Max Lanes Per Track, r=5}
  FunctionOutput(None=Floor(None=Divide(A=@MaterialExpressionMaterialFunctionCall_3.R, B=@n2))) {outputName=TrackIndex, sortPriority=1}
  FunctionOutput(None=Fmod(A=@MaterialExpressionMaterialFunctionCall_3.R, B=@n2)) {outputName=LaneIndex}
  n3 = Custom!(BitFlags=@MaterialExpressionMaterialFunctionCall_3.G) {
   code="int Mask = asint(BitFlags);\r\nfloat useProColors = (float)((Mask >> 0) && 1u);\r\nfloat isOverdrive  = (float)((Mask >> 1) && 1u);\r\nfloat useWorldPositionOffsets  = (float)((Mask >> 2) && 1u);\r\nreturn float3(useProColors, isOverdrive, useWorldPositionOffsets);"
   showCode=true
   inputs [{
     inputName=BitFlags
     input {expression=&self:MaterialExpressionMaterialFunctionCall_3, outputIndex=3}
    }]
  }
  FunctionOutput(None=ComponentMask(None=@n3) {b=true}) {
   description=Whether or not to use world position offsets instead of setting local position
   outputName=UseWorldPositionOffsets
   sortPriority=4
  }
  FunctionOutput(None=ComponentMask(None=@n3) {r=true}) {description="Whether or not this item should use \"pro\" colors. Value is 0/1. ", outputName=UseProColors, sortPriority=2}
  FunctionOutput(None=ComponentMask(None=@n3) {g=true}) {description="Whether or not this item is part of an overdrive phrase. Value is 0/1. ", outputName=IsOverdrive, sortPriority=3}
  FunctionOutput(None=@MaterialExpressionMaterialFunctionCall_3.A) {outputName=Distance From Now Bar, sortPriority=5}
 ]}
