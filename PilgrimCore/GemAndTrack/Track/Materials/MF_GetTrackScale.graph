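# Material function that returns the XYZ scale of a gem track, given a TrackIndex input.
# It reads the TrackScaleX, TrackScaleY and TrackScaleZ vectors from the MPC_Sparks parameter collection, each packing one axis for several tracks.
# MF_SelectComponentByIndex picks the indexed component from each vector, and the three results combine into the Vector3 Scale output.
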MaterialFunction /PilgrimCore/GemAndTrack/Track/Materials/MF_GetTrackScale
graph {expressions {
  MaterialExpressionCollectionParameter_0 {class=MaterialExpressionCollectionParameter, inputs [], properties {collection=/SparksEffects/Effects/MPC_Sparks, parameterName=TrackScaleZ}}
  MaterialExpressionCollectionParameter_2 {class=MaterialExpressionCollectionParameter, inputs [], properties {collection=&MPC_Sparks, parameterName=TrackScaleY}}
  MaterialExpressionCollectionParameter_3 {class=MaterialExpressionCollectionParameter, inputs [], properties {collection=&MPC_Sparks, parameterName=TrackScaleX}}
  MaterialExpressionConvert_0 {
   class=MaterialExpressionConvert
   inputs [
    None {from=MaterialExpressionMaterialFunctionCall_11, output=Value}
    None {from=MaterialExpressionMaterialFunctionCall_10, output=Value}
    None {from=MaterialExpressionMaterialFunctionCall_9, output=Value}
   ]
   properties {
    nodeName=Make Vector3
    convertInputs [
     {
      type=Scalar
      expressionInput {
       expression=&self:MaterialExpressionMaterialFunctionCall_11
       inputName=None
       mask=0
       maskA=0
       maskB=0
       maskG=0
       maskR=0
       outputIndex=0
      }
     }
     {
      type=Scalar
      expressionInput {
       expression=&self:MaterialExpressionMaterialFunctionCall_10
       inputName=None
       mask=0
       maskA=0
       maskB=0
       maskG=0
       maskR=0
       outputIndex=0
      }
     }
     {
      type=Scalar
      expressionInput {
       expression=&self:MaterialExpressionMaterialFunctionCall_9
       inputName=None
       mask=0
       maskA=0
       maskB=0
       maskG=0
       maskR=0
       outputIndex=0
      }
     }
    ]
    convertOutputs [{type=Vector3}]
   }
  }
  MaterialExpressionFunctionInput_0 {
   class=MaterialExpressionFunctionInput
   inputs [Preview {from="", output=""}]
   properties {inputName=TrackIndex, inputType=FunctionInput_Scalar}
  }
  MaterialExpressionFunctionOutput_0 {class=MaterialExpressionFunctionOutput, inputs [None {from=MaterialExpressionConvert_0, output=""}], properties {outputName=Scale}}
  MaterialExpressionMaterialFunctionCall_10 {
   class=MaterialExpressionMaterialFunctionCall
   inputs [
    Values {from=MaterialExpressionCollectionParameter_2}
    Index {from=MaterialExpressionReroute_0}
   ]
   properties {desc=Track Scale Y, materialFunction=/PilgrimCore/GemAndTrack/Gems/Materials/MF_SelectComponentByIndex}
  }
  MaterialExpressionMaterialFunctionCall_11 {
   class=MaterialExpressionMaterialFunctionCall
   inputs [
    Values {from=MaterialExpressionCollectionParameter_3}
    Index {from=MaterialExpressionReroute_2}
   ]
   properties {desc=Track Scale X, materialFunction=&MF_SelectComponentByIndex}
  }
  MaterialExpressionMaterialFunctionCall_9 {
   class=MaterialExpressionMaterialFunctionCall
   inputs [
    Values {from=MaterialExpressionCollectionParameter_0}
    Index {from=MaterialExpressionReroute_1}
   ]
   properties {desc=Track Scale Z, materialFunction=&MF_SelectComponentByIndex}
  }
  MaterialExpressionReroute_0 {class=MaterialExpressionReroute, inputs [None {from=MaterialExpressionFunctionInput_0}]}
  MaterialExpressionReroute_1 {class=MaterialExpressionReroute, inputs [None {from=MaterialExpressionReroute_0}]}
  MaterialExpressionReroute_2 {class=MaterialExpressionReroute, inputs [None {from=MaterialExpressionReroute_0}]}
 }}
