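# Material function for track items that outputs a world position offset sliding the item along its track by its distance from the now bar.
# A constant local Y axis is transformed and normalized, then multiplied by the negated distance and by the scale MF_GetTrackScale returns for the TrackIndex.
# A PreviousFrameSwitch selects current or previous-frame distance, and an If zeroes the offset when UseWorldPositionOffsets is below its threshold.
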MaterialFunction /PilgrimCore/GemAndTrack/Track/Materials/MF_PilgrimTrackItemWPO
graph {expressions {
  MaterialExpressionConstant_1 {class=MaterialExpressionConstant}
  MaterialExpressionConvert_0 {
   class=MaterialExpressionConvert
   inputs [
    None {from=""}
    None {from=""}
    None {from=""}
   ]
   properties {
    nodeName=Make Vector3
    convertInputs [
     {type=Scalar, expressionInput {expression=null, inputName=None, mask=0, maskA=0, maskB=0, maskG=0, maskR=0, outputIndex=0}}
     {
      type=Scalar
      defaultValue=<rgba 1 0 0 1>
      expressionInput {expression=null, inputName=None, mask=0, maskA=0, maskB=0, maskG=0, maskR=0, outputIndex=0}
     }
     {type=Scalar, expressionInput {expression=null, inputName=None, mask=0, maskA=0, maskB=0, maskG=0, maskR=0, outputIndex=0}}
    ]
    convertOutputs [{type=Vector3}]
   }
  }
  MaterialExpressionFunctionInput_0 {
   class=MaterialExpressionFunctionInput
   inputs [Preview {from="", output=""}]
   properties {inputName=PreviousFrameDistanceToNowBar, inputType=FunctionInput_Scalar, sortPriority=2}
  }
  MaterialExpressionFunctionInput_1 {
   class=MaterialExpressionFunctionInput
   inputs [Preview {from="", output=""}]
   properties {inputName=UseWorldPositionOffsets, inputType=FunctionInput_Scalar}
  }
  MaterialExpressionFunctionInput_2 {
   class=MaterialExpressionFunctionInput
   inputs [Preview {from="", output=""}]
   properties {inputName=TrackIndex, inputType=FunctionInput_Scalar}
  }
  MaterialExpressionFunctionInput_3 {
   class=MaterialExpressionFunctionInput
   inputs [Preview {from="", output=""}]
   properties {inputName=DistanceFromNowBar, inputType=FunctionInput_Scalar, sortPriority=1}
  }
  MaterialExpressionFunctionOutput_0 {
   class=MaterialExpressionFunctionOutput
   inputs [None {from=MaterialExpressionMultiply_1, output=""}]
   properties {outputName=Track Position Offset}
  }
  MaterialExpressionIf_2 {
   class=MaterialExpressionIf
   inputs [
    A {from=MaterialExpressionFunctionInput_1, output=""}
    B {from="", output=""}
    "A > B" {from=MaterialExpressionMultiply_12, output=""}
    "A == B" {from="", output=""}
    "A < B" {from=MaterialExpressionConstant_1, output=""}
   ]
   properties {constB=0.1}
  }
  MaterialExpressionMaterialFunctionCall_12 {
   class=MaterialExpressionMaterialFunctionCall
   inputs [TrackIndex {from=MaterialExpressionFunctionInput_2}]
   properties {materialFunction=&/MF_GetTrackScale}
  }
  MaterialExpressionMultiply_0 {
   class=MaterialExpressionMultiply
   inputs [
    A {from=MaterialExpressionNormalize_1}
    B {from=MaterialExpressionReroute_0}
   ]
  }
  MaterialExpressionMultiply_1 {
   class=MaterialExpressionMultiply
   inputs [
    A {from=MaterialExpressionMultiply_0}
    B {from=MaterialExpressionMaterialFunctionCall_12, output=Scale}
   ]
  }
  MaterialExpressionMultiply_12 {
   class=MaterialExpressionMultiply
   inputs [
    A {from=MaterialExpressionPreviousFrameSwitch_0}
    B {from=""}
   ]
   properties {constB=-1}
  }
  MaterialExpressionNormalize_1 {class=MaterialExpressionNormalize, inputs [VectorInput {from=MaterialExpressionTransform_5, output=""}]}
  MaterialExpressionPreviousFrameSwitch_0 {
   class=MaterialExpressionPreviousFrameSwitch
   inputs [
    Current Frame {from=MaterialExpressionFunctionInput_3, output=""}
    Previous Frame {from=MaterialExpressionFunctionInput_0, output=""}
   ]
   properties {}
  }
  MaterialExpressionReroute_0 {class=MaterialExpressionReroute, inputs [None {from=MaterialExpressionIf_2}]}
  MaterialExpressionTransform_5 {class=MaterialExpressionTransform, inputs [None {from=MaterialExpressionConvert_0}], properties {transformSourceType=TRANSFORMSOURCE_Local}}
 }}
