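# M_BeatMarker is an unlit additive material for beat markers on the Pilgrim gem track.
# Custom primitive data supplies track item parameters for color mode, track index, world position offset and now-bar distance fade.
# A sine across texture U creates a center-peaked glow that shapes the two-color emissive and the opacity.
# Opacity combines that shape with a per-marker value and a shared track phase selected by index from the MPC_Sparks parameter collection.
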Material /PilgrimCore/GemAndTrack/Track/Materials/M_BeatMarker
graph {
 properties {
  bEnableResponsiveAA=true
  bOutputTranslucentVelocity=true
  blendMode=BLEND_Additive
  pixelDepthOffsetMode=PDOM_Legacy
  shadingModel=MSM_Unlit
  usedShadingModels=Unlit
 }
 nodes [
  n1 = LinearInterpolate(
   A=Constant3Vector() {constant=<rgba 0.0896 0.0224 0.224 1>}
   B=Constant3Vector() {constant=<rgba 0.0953 0.1275 0.224 1>}
   Alpha=NamedRerouteUsage()
  )
  n2 = TextureCoordinate()
  n3 = Sine(None=Multiply(A=ComponentMask(None=@n2) {r=true}) {constB=0.5})
  n4 = Multiply(A=Add(A=@n1) {constB=0.2}, B=Power(Base=@n3))
  MaterialFunctionCall(Emissive=Add(A=@n1, B=Multiply(A=@n4) {constB=1.5})) {materialFunction=&MF_FX_EmissiveAdjustments}
  n5 = CollectionParameter() {collection=/SparksEffects/Effects/MPC_Sparks, parameterName=TrackSecondaryElementsPhase}
  n6 = NamedRerouteUsage()
  n7 = Constant() {r=0.05}
  n8 = NamedRerouteUsage()
  Multiply(
   A=@n3
   B=Multiply(
    A=MaterialFunctionCall!(Values=@n5, Index=@n6) {desc=Shared opacity from track, materialFunction=/PilgrimCore/GemAndTrack/Gems/Materials/MF_SelectComponentByIndex}.Value
    B=LinearInterpolate(A=@n7, Alpha=@n8) {constB=0, desc=Individual beat marker opacity}
   )
  )
  n9 = VectorParameter() {bUseCustomPrimitiveData=true, parameterName=PilgrimTrackItemStartIndex}
  n10 = MaterialFunctionCall!(In=@n9.RGBA) {materialFunction=&/MF_GetPilgrimTrackItemParams}
  NamedRerouteDeclaration(None=@n10.UseProColors) {name=UseProColors, nodeColor=<rgba 0 1 0.1094 1>}
  NamedRerouteDeclaration(None=MaterialFunctionCall!(FadeDistance, DistanceFromNowBar=@n9.A) {materialFunction=&/MF_GetPilgrimBeatMarkerNowBarToAlpha}.Alpha) {name=Alpha, nodeColor=<rgba 0 1 0.1094 1>}
  NamedRerouteDeclaration(None=MaterialFunctionCall!(
    TrackIndex=NamedRerouteDeclaration(None=@n10.TrackIndex) {name=TrackIndex, nodeColor=<rgba 0 1 0.1094 1>}
    UseWorldPositionOffsets=@n10.UseWorldPositionOffsets
    DistanceFromNowBar=@n10.UseWorldPositionOffsets
    PreviousFrameDistanceToNowBar=@n10.UseWorldPositionOffsets
   ) {materialFunction=&/MF_PilgrimTrackItemWPO}.Track Position Offset) {name=Track Position Offset, nodeColor=<rgba 0 1 0.1094 1>}
  NamedRerouteUsage()
 ]
}
