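# Translucent unlit, two-sided material for the smasher on the Pilgrim gem track, with parameter switches for drum and mobile-portrait variants.
# Vertex-color masks, radial gradients and UV falloffs build outline and glow, dropping vertical glow on mobile portrait.
# Custom primitive data supplies glow state and track/lane info, driving curve-atlas lane colors, which Overdrive overrides.
# A shared MPC_Sparks parameter supplies per-track alpha, and MF_FX_EmissiveAdjustments processes emissive.
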Material /PilgrimCore/GemAndTrack/Track/Materials/M_Smasher
graph {
 properties {
  blendMode=BLEND_Translucent
  pixelDepthOffsetMode=PDOM_Legacy
  previewMesh=/Engine/Transient.MaterialEditorMeshComponent_10
  shadingModel=MSM_Unlit
  twoSided=true
  usedShadingModels=Unlit
 }
 nodes [
  n1 = TextureCoordinate() {coordinateIndex=1}
  n2 = Sine(None=Multiply(A=ComponentMask(None=@n1) {r=true}) {constB=0.5})
  n3 = SmoothStep(Min=ScalarParameter(), Max=ScalarParameter() {defaultValue=1, parameterName=Param_1}, Value=OneMinus(None=ConstantBiasScale(None=@n2)))
  Add(
   A=@n3
   B=MaterialFunctionCall!(
    Width=ScalarParameter() {defaultValue=0.1, parameterName=Param_2}
    Sharpness=ScalarParameter() {defaultValue=1, parameterName=Param_3}
    Offset
    Direction Switch
    Compare
    Input Coordinates=TextureCoordinate() {coordinateIndex=1}
   ) {materialFunction=/Engine/Functions/Engine_MaterialFunctions03/Procedurals/GeneratedBand}.Result
  )
  n4 = NamedRerouteUsage()
  n5 = ScalarParameter() {parameterName=IsDrum}
  n6 = NamedRerouteUsage()
  n7 = VertexColor()
  n8 = SmoothStep(Max=Constant() {r=0.3}, Value=@n7.B) {constMax=0.1}
  n9 = Multiply(A=Add(A=Multiply(A=@n6, B=@n8), B=NamedRerouteUsage()), B=NamedRerouteUsage())
  n10 = ComponentMask(None=TextureCoordinate()) {r=true}
  n11 = Constant() {r=1}
  n12 = Multiply(A=Power(Base=OneMinus(None=@n10), Exp=Constant() {r=4}) {constExponent=3}, B=OneMinus(None=Fresnel(ExponentIn=@n11) {exponent=1}))
  n13 = SmoothStep(Value=@n10) {constMax=0.05}
  n14 = Add(A=@n9, B=Multiply(A=Multiply(A=@n12, B=@n13) {constB=0.5}, B=NamedRerouteUsage()))
  n15 = ScalarParameter() {parameterName=IsMobilePortrait}
  n16 = Multiply(
   A=Add(
    A=LinearInterpolate(A=@n4, Alpha=@n5) {constA=1, constB=0}
    B=LinearInterpolate(A=@n14, B=@n9, Alpha=@n15) {desc=Looks cleaner without vertical glow on Mobile}
   )
   B=NamedRerouteUsage()
  )
  n17 = Add(A=@n3, B=ScalarParameter() {defaultValue=0.05, parameterName=Param_5})
  LinearInterpolate(A=@n16, B=Multiply(A=@n16, B=Saturate(None=@n17)), Alpha=ScalarParameter() {parameterName=IsDrum}) {constA=1}
  n18 = Multiply(A=NamedRerouteUsage()) {constB=3}
  n19 = NamedRerouteUsage()
  n20 = NamedRerouteUsage()
  n21 = Multiply(A=Multiply(A=NamedRerouteUsage(), B=@n20), B=NamedRerouteUsage())
  n22 = ComponentMask(None=TextureCoordinate()) {g=true}
  n23 = Multiply(A=Multiply(A=@n20, B=OneMinus(None=@n22)), B=NamedRerouteUsage())
  n24 = NamedRerouteUsage()
  n25 = NamedRerouteUsage()
  n26 = Add(
   A=LinearInterpolate(A=Multiply(A=@n18, B=@n19), Alpha=ScalarParameter() {parameterName=IsDrum}) {constA=1, constB=0}
   B=Add(A=Add(A=@n21, B=@n23), B=Multiply(A=@n24, B=@n25))
  )
  n27 = Multiply(A=@n26, B=ScalarParameter() {defaultValue=5, parameterName=Param_4}) {constB=10}
  n28 = NamedRerouteUsage()
  MaterialFunctionCall(Emissive=LinearInterpolate(A=@n26, B=LinearInterpolate(A=@n26, B=@n27, Alpha=@n28), Alpha=ScalarParameter() {parameterName=IsDrum}) {constA=1}) {materialFunction=&MF_FX_EmissiveAdjustments}
  n29 = VertexColor()
  NamedRerouteDeclaration(None=Add(A=@n29.R, B=@n29.R)) {name=Bottom Mask, nodeColor=<rgba 0 0 0 1>}
  n30 = Constant() {r=1.25}
  n31 = Constant() {r=1.5}
  NamedRerouteDeclaration(None=OneMinus(None=MaterialFunctionCall(Radius=@n30, Density=@n31) {materialFunction=&RadialGradientExponential}.RadialGradientExponential)) {name=Radial Glow, nodeColor=<rgba 0 0 0 1>}
  n32 = NamedRerouteUsage()
  NamedRerouteDeclaration(None=LinearInterpolate(B=Add(A=@n32), Alpha=NamedRerouteUsage()) {desc="Non-pro colors use index 0\r\nPro colors use lane index shifted up by 1, to indices 1-5"}) {name=LaneColorIndex, nodeColor=<rgba 0 1 0.0291 1>}
  NamedRerouteDeclaration(None=MaterialFunctionCall!(
    Values=CollectionParameter() {collection=/SparksEffects/Effects/MPC_Sparks, parameterName=TrackSecondaryElementsPhase}
    Index=NamedRerouteUsage()
   ) {desc=Shared opacity from track, materialFunction=/PilgrimCore/GemAndTrack/Gems/Materials/MF_SelectComponentByIndex}.Value) {name=Alpha, nodeColor=<rgba 0 0 0 1>}
  NamedRerouteDeclaration(None=Multiply(A=SmoothStep(Value=@n7.R) {constMax=0.5}, B=@n8)) {name=Outline Mask, nodeColor=<rgba 0 0 0 1>}
  NamedRerouteDeclaration(None=@n29.G) {name=Upwards Mask, nodeColor=<rgba 0 0 0 1>}
  n33 = MaterialFunctionCall(Radius=Constant() {r=0.5}, Density=Constant() {r=1.5}) {materialFunction=&RadialGradientExponential}
  n34 = NamedRerouteUsage()
  NamedRerouteDeclaration(None=Multiply(A=NamedRerouteUsage(), B=Multiply(A=@n33.RadialGradientExponential, B=@n34))) {name=Anim, nodeColor=<rgba 0 0 0 1>}
  n35 = Subtract(B=ScalarParameter() {parameterName=Overdrive Active})
  n36 = NamedRerouteUsage()
  n37 = CurveAtlasRowParameter(CurveTime=Divide(A=Add(A=@n36) {constB=0.5}) {constB=6}) {
   atlas=/PilgrimCore/GemAndTrack/Colors/Sparks_Track_LUT
   curve=/PilgrimCore/GemAndTrack/Colors/Sparks_Track_Smasher
   parameterName=SmasherColor
  }
  n38 = ScalarParameter() {parameterName=Overdrive Active}
  n39 = VectorParameter() {parameterName=Overdrive Color, defaultValue=<rgba 1 0.2736 0 1>}
  NamedRerouteDeclaration(None=Add(A=Multiply(A=@n35, B=@n37), B=Multiply(A=@n38, B=@n39.RGB))) {name=Color, nodeColor=<rgba 0 0 0 1>}
  NamedRerouteDeclaration(None=ScalarParameter() {bUseCustomPrimitiveData=true, parameterName=GlowOn, primitiveDataIndex=3}) {name=GlowOn, nodeColor=<rgba 0 1 0.0291 1>}
  n40 = VectorParameter() {bUseCustomPrimitiveData=true, parameterName=PilgrimTrackItemStartIndex, channelNames {a=DistanceFromNowBar, b="", g="", r=""}}
  n41 = MaterialFunctionCall!(In=@n40.RGBA) {materialFunction=&/MF_GetPilgrimTrackItemParams}
  NamedRerouteDeclaration(None=@n41.TrackIndex) {name=TrackIndex, nodeColor=<rgba 0 1 0.0291 1>}
  NamedRerouteDeclaration(None=@n41.LaneIndex) {name=LaneIndex, nodeColor=<rgba 0 1 0.0291 1>}
  NamedRerouteDeclaration(None=@n41.UseProColors) {name=UseProColors, nodeColor=<rgba 0 1 0.0291 1>}
  Step(X=@n29.R) {constX=0.1, constY=0.2}
 ]
}
