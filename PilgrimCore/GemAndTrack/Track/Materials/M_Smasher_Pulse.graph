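# Translucent, unlit, two-sided material for a pulse effect on Pilgrim track smashers.
# Custom primitive data sets pulse start and duration for an eased-out animation value, while a narrow band masked from red vertex color fades as red rises.
# That mask is scaled by shared track opacity from MPC_Sparks, then blended by IsDrum.
# Emissive color samples a curve atlas, with a lane color index derived from track item parameters and pro-color mode.
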Material /PilgrimCore/GemAndTrack/Track/Materials/M_Smasher_Pulse
graph {
 properties {
  blendMode=BLEND_Translucent
  pixelDepthOffsetMode=PDOM_Legacy
  previewMesh=/Engine/Transient.MaterialEditorMeshComponent_24
  shadingModel=MSM_Unlit
  twoSided=true
  usedShadingModels=Unlit
 }
 nodes [
  Frac(None=Time())
  n1 = NamedRerouteUsage()
  n2 = Subtract(A=VertexColor().R, B=@n1)
  n3 = Constant() {r=0.01}
  n4 = Divide(A=Subtract(A=VertexColor().R, B=Add(A=@n1) {constB=0.1}), B=@n3) {constB=0.1}
  n5 = Multiply(A=Saturate(None=Divide(A=@n2, B=@n3) {constB=0.1}), B=Saturate(None=OneMinus(None=@n4)))
  n6 = OneMinus(None=VertexColor().R)
  n7 = CollectionParameter() {collection=/SparksEffects/Effects/MPC_Sparks, parameterName=TrackSecondaryElementsPhase}
  n8 = NamedRerouteUsage()
  LinearInterpolate(
   A=Multiply(
    A=Multiply(A=@n5, B=@n6)
    B=MaterialFunctionCall!(Values=@n7, Index=@n8) {desc=Shared opacity from track, materialFunction=/PilgrimCore/GemAndTrack/Gems/Materials/MF_SelectComponentByIndex}.Value
   )
   Alpha=ScalarParameter() {parameterName=IsDrum}
  ) {constB=0}
  n9 = CurveAtlasRowParameter(CurveTime=Divide(A=NamedRerouteUsage()) {constB=6}) {
   atlas=/PilgrimCore/GemAndTrack/Colors/Sparks_Track_LUT
   curve=/PilgrimCore/GemAndTrack/Colors/Sparks_Track_Smasher
   parameterName=SmasherColor
  }
  MaterialFunctionCall(Emissive=Multiply(A=@n9) {constB=30}) {materialFunction=&MF_FX_EmissiveAdjustments}
  n10 = ComponentMask(None=MaterialFunctionCall() {materialFunction=&BoundingBoxBased_0-1_UVW}.G) {r=true}
  n11 = Power(Base=@n10)
  Multiply(A=@n5, B=@n11)
  n12 = Constant() {r=1}
  n13 = Divide(
   A=Subtract(A=Time(), B=ScalarParameter() {bUseCustomPrimitiveData=true, parameterName=PulseStartTime, primitiveDataIndex=4})
   B=ScalarParameter() {bUseCustomPrimitiveData=true, parameterName=PulseDuration, primitiveDataIndex=5}
  )
  n14 = Power(Base=Subtract(A=@n12, B=Saturate(None=@n13))) {constExponent=3}
  NamedRerouteDeclaration(None=LinearInterpolate(Alpha=Subtract(A=@n12, B=@n14))) {name=PulseAnim, nodeColor=<rgba 0 1 0.0291 1>}
  n15 = NamedRerouteUsage()
  NamedRerouteDeclaration(None=LinearInterpolate(B=Add(A=@n15), Alpha=NamedRerouteUsage()) {desc="Non-pro colors use index 0\r\nPro colors use lane index shifted up by 1, to indices 1-5"}) {name=LaneColorIndex, nodeColor=<rgba 0 1 0.0291 1>}
  n16 = MaterialFunctionCall!(In=VectorParameter() {bUseCustomPrimitiveData=true, parameterName=PilgrimTrackItemStartIndex}.RGBA) {materialFunction=&/MF_GetPilgrimTrackItemParams}
  NamedRerouteDeclaration(None=@n16.TrackIndex) {name=TrackIndex, nodeColor=<rgba 0 1 0.0291 1>}
  NamedRerouteDeclaration(None=@n16.LaneIndex) {name=LaneIndex, nodeColor=<rgba 0 1 0.0291 1>}
  NamedRerouteDeclaration(None=@n16.UseProColors) {name=UseProColors, nodeColor=<rgba 0 1 0.0291 1>}
  OneMinus(None=ConstantBiasScale(None=@n10))
  n17 = ComponentMask(None=TextureCoordinate() {coordinateIndex=1}) {g=true}
  Power(Base=ConstantBiasScale(None=OneMinus(None=@n17))) {constExponent=10}
  Saturate(None=Multiply(A=@n11) {constB=50})
  Sine(None=Multiply(A=@n17) {constB=0.5})
  Sine(None=Multiply(A=@n10) {constB=0.5})
  VertexColor()
 ]
}
