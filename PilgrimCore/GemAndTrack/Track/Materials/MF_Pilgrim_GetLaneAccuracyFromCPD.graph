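# Material function for the Pilgrim gem track that reads a lane's opacity and accuracy from Custom Primitive Data.
# A Custom HLSL node takes LaneIndex, clamps it to five lanes, and reads opacity from floats 0-4 and accuracy from floats 5-9.
# It copies only the first three CPD vectors into a local array, so variable indexing does not block optimization of GetPrimitiveData's copy loop.
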MaterialFunction /PilgrimCore/GemAndTrack/Track/Materials/MF_Pilgrim_GetLaneAccuracyFromCPD
graph {nodes [
  MaterialExpressionFunctionInput_0 = FunctionInput() {inputName=LaneIndex, inputType=FunctionInput_Scalar}
  n1 = Custom!(LaneIndexF=@MaterialExpressionFunctionInput_0) {
   code="// Make a copy of just the first 3 vectors into a local array. Directly accessing CustomPrimitiveData with\r\n// a variable index prevents optimization of the array copying loop inside GetPrimitiveData. This lets us\r\n// copy only the values within our access bounds of [0-9]\r\nFPrimitiveSceneData PrimitiveData = GetPrimitiveData(Parameters);\r\nfloat4 LanePrimitiveData[3] = { \r\n    PrimitiveData.CustomPrimitiveData[0],\r\n    PrimitiveData.CustomPrimitiveData[1],\r\n    PrimitiveData.CustomPrimitiveData[2]\r\n};\r\n\r\n// Calculate indices for opacity [0-4] and accuracy [5-9]\r\nuint LaneIndex = clamp((uint)LaneIndexF, 0, 4);\r\nuint OpacityIndex = LaneIndex;\r\nuint AccuracyIndex = LaneIndex + 5;\r\n\r\n// Read data from CPD\r\nfloat Opacity = LanePrimitiveData[OpacityIndex / 4][OpacityIndex % 4];\r\nOutAccuracy = LanePrimitiveData[AccuracyIndex / 4][AccuracyIndex % 4];\r\n\r\nreturn Opacity;"
   outputType=CMOT_Float1
   showCode=true
   additionalOutputs [{outputName=OutAccuracy, outputType=CMOT_Float1}]
   inputs [{
     inputName=LaneIndexF
     input {expression=&self:MaterialExpressionFunctionInput_0}
    }]
  }
  FunctionOutput(None=@n1.return) {outputName=Opacity}
  FunctionOutput(None=@n1.OutAccuracy) {outputName=Accuracy, sortPriority=1}
 ]}
