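# Material function for Pilgrim track lanes that turns hit accuracy (late, perfect or early) into a timed color flash and a mask.
# Accuracy data packs a start time, accuracy value and lane mask; elapsed time since the start drives an opacity fade-out.
# If nodes choose Perfect or Good color and the perfect, early or late gradient channel, boosting early and late by GoodsBoost.
# Named reroutes carry these values into the multiplied Color and LaneMask outputs.
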MaterialFunction /PilgrimCore/GemAndTrack/Track/Materials/MF_Pilgrim_LaneAccuracyOptimized
graph {
 properties {bExposeToLibrary=true, libraryCategoriesText [Pilgrim]}
 nodes [
  n1 = Multiply(A=NamedRerouteUsage(), B=NamedRerouteUsage())
  n2 = NamedRerouteUsage()
  FunctionOutput(None=Multiply(A=Multiply(A=@n1, B=@n2), B=NamedRerouteUsage())) {outputName=Color_1}
  n3 = NamedRerouteUsage()
  n4 = NamedRerouteUsage()
  FunctionOutput(None=Multiply(A=Multiply(A=@n3, B=@n4), B=NamedRerouteUsage())) {outputName=LaneMask_1, sortPriority=5}
  NamedRerouteDeclaration(None=FunctionInput() {inputName=GoodColor_1, sortPriority=2, previewValue=<0 0 1 1>}) {name=GoodColor, nodeColor=<rgba 0 1 0.5781 1>}
  NamedRerouteDeclaration(None=FunctionInput() {inputName=ParfectColor_1, sortPriority=1, previewValue=<1 1 0 1>}) {name=PerfectColor, nodeColor=<rgba 1 0.2578 0 1>}
  n5 = NamedRerouteUsage()
  NamedRerouteDeclaration(None=If(A=NamedRerouteUsage(), "A > B"=@n5, "A == B"=NamedRerouteUsage(), "A < B"=@n5) {equalsThreshold=0.1}) {name=FinalColor, nodeColor=<rgba 0 0.0625 1 1>}
  n6 = FunctionInput() {inputName=GradientData}
  NamedRerouteDeclaration(None=ComponentMask(None=@n6) {g=true}) {name=EarlyGradient, nodeColor=<rgba 0.3828 1 0 1>}
  NamedRerouteDeclaration(None=ComponentMask(None=@n6) {b=true}) {name=LateGradient, nodeColor=<rgba 1 0 0.7031 1>}
  NamedRerouteDeclaration(None=ComponentMask(None=@n6) {r=true}) {name=PerfectGradient, nodeColor=<rgba 0 1 0.9766 1>}
  n7 = NamedRerouteUsage()
  n8 = ScalarParameter() {defaultValue=1.5, parameterName=GoodsBoost}
  n9 = NamedRerouteUsage()
  NamedRerouteDeclaration(None=If(
    A=NamedRerouteUsage()
    "A > B"=Multiply(A=@n7, B=@n8) {constB=1.5, desc=Boost Goods Intensity}
    "A == B"=NamedRerouteUsage()
    "A < B"=Multiply(A=@n9, B=@n8) {constB=1.5, desc=Boost Goods Intensity}
   ) {equalsThreshold=0.1}) {name=FinalGradient, nodeColor=<rgba 1 0.6562 0 1>}
  n10 = FunctionInput() {inputName=AccuracyAndMaskData}
  NamedRerouteDeclaration(None=ComponentMask(None=@n10) {b=true}) {name=LaneMask, nodeColor=<rgba 0.3359 0 1 1>}
  NamedRerouteDeclaration(None=ComponentMask(None=@n10) {desc="0 == Late, 0.5 == Perfect, 1 == Early", g=true}) {name=Accuracy, desc="-1 == Late, 0 == Perfect, 1 == Early", nodeColor=<rgba 0 1 0.0156 1>}
  n11 = Divide(A=Subtract(A=Time(), B=ComponentMask(None=@n10) {r=true})) {constB=0.6}
  NamedRerouteDeclaration(None=OneMinus(None=Saturate(None=@n11))) {name=Opacity, nodeColor=<rgba 1 0 0.3047 1>}
 ]
}
