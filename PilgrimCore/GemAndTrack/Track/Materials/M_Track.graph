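# Unlit, masked, two-sided material for the rhythm-game note track, with plastic-instrument, inverse-track and 4/5-lane variants.
# Named reroutes build layers (base, rails, dividers, overdrive, max-streak peak state, sparkles) that are summed into emissive adjustments.
# Per-lane accuracy feedback uses CPD lookup functions and UV-derived lane indices, treating the drum kit as lane index 4.
# Pivot Painter data and a phase parameter drive switchable world-position-offset piece animation with Fresnel glow.
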Material /PilgrimCore/GemAndTrack/Track/Materials/M_Track
graph {
 properties {
  bEnableResponsiveAA=true
  blendMode=BLEND_Masked
  opacityMaskClipValue=0.01
  pixelDepthOffsetMode=PDOM_Legacy
  shadingModel=MSM_Unlit
  twoSided=true
  usedShadingModels=Unlit
 }
 nodes [
  Add()
  n1 = MaterialFunctionCall() {materialFunction=&LinearGradient}
  n2 = Sine(None=@n1.VGradient)
  ConstantBiasScale(None=@n2)
  Constant() {r=1}
  n3 = VectorParameter() {parameterName=OverdriveColor, defaultValue=<rgba 1 0.2421 0.05 1>}
  LinearInterpolate(
   A=@n3.RGB
   B=VectorParameter() {parameterName=OverdriveColorPlastic, defaultValue=<rgba 0.6 0.6 0.6 1>}.RGB
   Alpha=ScalarParameter() {parameterName=IsPlasticOverdrive}
  )
  MaterialFunctionCall(
   Color_Morning=Constant3Vector() {constant=<rgba 0.0136 0.01 0.04 1>}
   Color_Day=Constant3Vector() {constant=<rgba 0.0158 0.0113 0.05 1>}
   Color_Evening=Constant3Vector() {constant=<rgba 0.0167 0.01 0.05 1>}
   Color_Night=Constant3Vector() {constant=<rgba 0.0185 0.0079 0.025 1>}
  ) {materialFunction=&Color_Time_of_Day}
  n4 = Add(A=NamedRerouteUsage(), B=NamedRerouteUsage()) {desc=OD Flames + BG + Max Streak}
  n5 = NamedRerouteUsage()
  n6 = Add(A=Add(A=Add(A=@n4, B=@n5), B=NamedRerouteUsage()), B=NamedRerouteUsage())
  n7 = NamedRerouteUsage()
  MaterialFunctionCall(Emissive=Add(A=Add(A=@n6, B=@n7), B=NamedRerouteUsage())) {materialFunction=&MF_FX_EmissiveAdjustments}
  MaterialFunctionCall(
   Color_Morning=Constant3Vector() {constant=<rgba 0.15 0.1641 1 1>}
   Color_Day=Constant3Vector() {constant=<rgba 0.15 0.1641 1 1>}
   Color_Evening=Constant3Vector() {constant=<rgba 0.15 0.1641 1 1>}
   Color_Night=Constant3Vector() {constant=<rgba 0.2515 0.1038 0.625 1>}
  ) {materialFunction=&Color_Time_of_Day}
  MaterialFunctionCall(
   Color_Morning=Constant3Vector() {constant=<rgba 1 0.2354 0.0417 1>}
   Color_Day=Constant3Vector() {constant=<rgba 1 0.2354 0.0417 1>}
   Color_Evening=Constant3Vector() {constant=<rgba 1 0.2354 0.0417 1>}
   Color_Night=Constant3Vector() {constant=<rgba 1 0.1875 0.0417 1>}
  ) {materialFunction=&Color_Time_of_Day}
  n8 = MaterialFunctionCall!(Repeat Count=Constant2Vector() {g=1, r=2}, texcoord=TextureCoordinate()) {materialFunction=/Game/Environments/Papaya/Materials/MaterialFunctions/MF_UVMirror}
  n9 = AppendVector(A=ScalarParameter() {defaultValue=1}, B=ScalarParameter() {defaultValue=3, parameterName=Param_1})
  n10 = TextureSample(UVs=Panner(Coordinate=Multiply(A=@n8.Result, B=@n9)) {speedX=-0.65, speedY=-0.25}) {texture=/PilgrimCore/GemAndTrack/Gems/Textures/T_Sparks_Overdrive_Sparkle}
  n11 = Sine(None=Multiply(A=MaterialFunctionCall() {materialFunction=&LinearGradient}.UGradient) {constB=0.5})
  NamedRerouteDeclaration(None=Multiply(A=Multiply(A=@n10.G) {constB=20}, B=SmoothStep(Value=@n11) {constMax=0, constMin=0.6})) {name=Sparkles, nodeColor=<rgba 0 0 0 1>}
  NamedRerouteDeclaration(None=LinearInterpolate(
    A=VectorParameter() {parameterName=TrackColor, defaultValue=<rgba 0.0167 0.01 0.05 1>}.RGB
    B=VectorParameter() {parameterName=TrackColorPlastic, defaultValue=<rgba 0.0128 0.0138 0.025 1>}.RGB
    Alpha=ScalarParameter() {parameterName=IsPlastic}
   )) {name=DefaultTrackColor, nodeColor=<rgba 1 0.4453 0 1>}
  n12 = Add(A=ComponentMask(None=TextureCoordinate()) {g=true}) {constB=-0.35}
  n13 = OneMinus(None=@n12)
  n14 = Step(Y=@n12) {constX=0.5}
  n15 = Subtract(A=SmoothStep(Min=Constant() {r=0.35}, Max=Constant() {r=0.65}, Value=LinearInterpolate(A=@n13, B=@n12, Alpha=@n14)) {constMax=0.65, constMin=0.35}) {constB=0.05}
  n16 = Multiply(A=Saturate(None=Power(Base=@n15) {constExponent=0.5}), B=@n14) {desc=Early gradient}
  n17 = Power(Base=@n15) {constExponent=2.5}
  n18 = Multiply(A=Saturate(None=Multiply(A=@n17) {constB=5}), B=OneMinus(None=@n14)) {desc=Late gradient}
  NamedRerouteDeclaration(None=MaterialFunctionCall(R=Add(A=@n16, B=@n18) {desc=Perfect gradient}, G=@n16, B=@n18) {materialFunction=&AppendMany}.RGB) {name=LaneAccuracyGradients, nodeColor=<rgba 0.4531 0 1 1>}
  n19 = NamedRerouteUsage()
  n20 = Power(Base=@n19) {constExponent=10}
  n21 = MaterialFunctionCall!(Base=NamedRerouteUsage(), Add=Saturate(None=Multiply(A=@n20) {constB=1000})) {materialFunction=/Engine/Functions/Engine_MaterialFunctions02/Math/MultiplyAdd}
  NamedRerouteDeclaration(None=Multiply(
    A=LinearInterpolate(B=@n21.Result, Alpha=@n19) {constAlpha=1}
    B=ScalarParameter() {defaultValue=1, group=PerLaneFeedback, parameterName=FeedbackOpacity, sliderMax=1}
   )) {name=AccuracyState, nodeColor=<rgba 0 0 0 1>}
  n22 = NamedRerouteUsage()
  n23 = NamedRerouteUsage()
  n24 = Constant2Vector() {g=1, r=1.25}
  n25 = MaterialFunctionCall!(Width=Constant() {r=0.018}, Sharpness=Constant() {r=0.8}, Offset=Constant() {r=0.6475}, Direction Switch, Compare, Input Coordinates) {materialFunction=/Engine/Functions/Engine_MaterialFunctions03/Procedurals/GeneratedBand}
  n26 = MaterialFunctionCall(Width=Constant() {r=0.018}, Sharpness=Constant() {r=0.8}, Offset=Constant() {r=0.65}) {materialFunction=&GeneratedBand}
  n27 = OneMinus(None=StaticSwitchParameter(True=Power(Base=@n25.Result), False=Power(Base=@n26.Result)) {parameterName=IsPilgrimInverseTrack})
  n28 = Constant() {r=0.65}
  n29 = Multiply(
   A=MaterialFunctionCall(
    UVs=MaterialFunctionCall(Texture Scale=@n24) {materialFunction=&ScaleUVsByCenter}.UVs
    CenterPosition=Constant2Vector() {g=1.075, r=0.5}
    Radius=Constant() {r=0.5}
    Density=Constant() {r=2.33}
   ) {materialFunction=&RadialGradientExponential}.RadialGradientExponential
   B=Saturate(None=Add(A=@n27, B=@n28) {constB=0})
  )
  n30 = Divide(A=Subtract(A=Multiply(A=@n23, B=@n29), B=Constant() {desc=0.1, r=0.1}), B=Constant() {r=5})
  n31 = Power(Base=@n29, Exp=Constant() {r=3})
  n32 = Constant() {desc=0.5, r=0.5}
  n33 = Multiply(
   A=Add(A=Saturate(None=@n30), B=Multiply(A=@n31, B=@n32))
   B=LinearInterpolate(
    A=VectorParameter() {parameterName=PeakStateColor, defaultValue=<rgba 0.1575 0.1575 0.9 1>}.RGB
    B=VectorParameter() {parameterName=PeakStateColorPlastic, defaultValue=<rgba 0.141 0.2174 0.5885 1>}.RGB
    Alpha=ScalarParameter() {parameterName=IsPlastic}
   )
  )
  n34 = NamedRerouteDeclaration(None=ScalarParameter() {parameterName=Max Streak On/Off}) {name=MaxStreakOnOff, nodeColor=<rgba 0 0 0 1>}
  NamedRerouteDeclaration(None=Multiply(A=OneMinus(None=@n22), B=Multiply(A=@n33, B=@n34))) {name=Peak State, nodeColor=<rgba 0 0 0 1>}
  n35 = MaterialFunctionCall!(LaneIndex=Constant() {desc="DrumKit is treated like the 5th lane (index 4)", r=4}) {materialFunction=&/MF_Pilgrim_GetLaneAccuracyFromCPD}
  n36 = NamedRerouteUsage()
  n37 = NamedRerouteUsage()
  n38 = MaterialFunctionCall!(
   GradientData=NamedRerouteUsage()
   AccuracyAndMaskData=MaterialFunctionCall(R=@n35.Opacity, G=@n35.Opacity, B=NamedRerouteUsage()) {materialFunction=&AppendMany}.RGB
   ParfectColor_1=@n36
   GoodColor_1=@n37
  ) {materialFunction=&/MF_Pilgrim_LaneAccuracyOptimized}
  n39 = NamedRerouteUsage()
  n40 = MaterialFunctionCall(LaneIndex=NamedRerouteUsage()) {materialFunction=&MF_Pilgrim_GetLaneAccuracyFromCPD}
  n41 = MaterialFunctionCall(
   GradientData=NamedRerouteUsage()
   AccuracyAndMaskData=MaterialFunctionCall(R=@n40.Opacity, G=@n40.Opacity, B=NamedRerouteUsage()) {materialFunction=&AppendMany}.RGB
   ParfectColor_1=@n36
   GoodColor_1=@n37
  ) {materialFunction=&MF_Pilgrim_LaneAccuracyOptimized}
  n42 = Max(A=@n41.LaneMask_1, B=Multiply(A=Multiply(A=@n38.LaneMask_1, B=@n39)) {constB=1.5}) {desc="Blend for the pixels that overlap "}
  NamedRerouteDeclaration(None=LinearInterpolate(A=Saturate(None=@n42), Alpha=NamedRerouteUsage()) {constA=1, constB=0}) {name=AccuracyMask, nodeColor=<rgba 1 0 0.5625 1>}
  NamedRerouteDeclaration(None=ComponentMask(None=TextureCoordinate()) {r=true}) {name=UV_X, nodeColor=<rgba 1 0.0938 0 1>}
  NamedRerouteDeclaration(None=ScalarParameter() {defaultValue=1, parameterName=IsDrumKit}) {name=IsDrumKit, nodeColor=<rgba 0 0.2266 1 1>}
  n43 = Add(A=Constant() {r=0.5}, B=@n2)
  n44 = Multiply(A=ComponentMask(None=TextureCoordinate()) {r=true}) {constB=0.5}
  n45 = Add(
   A=Add(A=Saturate(None=@n43), B=@n1.VGradient)
   B=SmoothStep(Min=Constant() {desc=0.15, r=0.14}, Max=Constant() {desc=0.13, r=0.141}, Value=Sine(None=@n44)) {constMax=0, constMin=0.1}
  )
  n46 = Constant() {desc=0.15, r=0.151}
  n47 = Constant() {desc=0.13, r=0.15}
  n48 = TextureCoordinate()
  n49 = Sine(None=Multiply(A=ComponentMask(None=@n48) {r=true}) {constB=0.5})
  n50 = NamedRerouteDeclaration(None=Power(Base=SmoothStep(Min=@n46, Max=@n47, Value=@n49) {constMax=0, constMin=0.1})) {name=RailMask, nodeColor=<rgba 0 0 0 1>}
  NamedRerouteDeclaration(None=Multiply(A=NamedRerouteUsage(), B=Multiply(A=@n45, B=@n50))) {name=Rails, nodeColor=<rgba 0 0 0 1>}
  n51 = Constant() {r=-0.1}
  n52 = Add(A=Constant() {r=0.7}, B=@n51)
  n53 = Add(A=@n51, B=Constant() {r=1})
  n54 = ComponentMask(None=TextureCoordinate()) {g=true}
  n55 = OneMinus(None=MaterialFunctionCall() {materialFunction=&LinearGradient}.VGradient)
  n56 = Constant() {r=2}
  n57 = Sine(None=Multiply(A=MaterialFunctionCall() {materialFunction=&LinearGradient}.UGradient) {constB=0.5})
  n58 = Multiply(A=Power(Base=OneMinus(None=@n57), Exp=Constant() {r=8}) {constExponent=4}, B=Add(A=NamedRerouteUsage(), B=NamedRerouteUsage()))
  n59 = Multiply(A=Multiply(A=Power(Base=@n55, Exp=@n56), B=Saturate(None=@n58)), B=Constant() {r=20})
  n60 = MaterialFunctionCall(Width=Constant() {r=0.5}, Sharpness=Constant(), Direction Switch=StaticBool() {value=true}) {materialFunction=&GeneratedBand}
  n61 = OneMinus(None=MaterialFunctionCall() {materialFunction=&LinearGradient}.VGradient)
  n62 = Multiply(A=Power(Base=OneMinus(None=@n60.Result), Exp=Constant() {r=6}), B=Saturate(None=Multiply(A=@n61) {constB=2}))
  n63 = NamedRerouteUsage()
  n64 = Multiply(A=OneMinus(None=SmoothStep(Min=@n52, Max=@n53, Value=@n54)), B=Multiply(A=Add(A=@n59, B=@n62), B=OneMinus(None=@n63)))
  n65 = NamedRerouteDeclaration(None=ScalarParameter() {desc="Overdrive Track Color On / Off Lerp (0-1)", group=Overdrive, parameterName=Overdrive On/Off, sliderMax=1}) {name=Overdrive Value, nodeColor=<rgba 0 0 0 1>}
  NamedRerouteDeclaration(None=Multiply(A=Multiply(A=@n64, B=@n65), B=@n3.RGB)) {name=Overdrive State, nodeColor=<rgba 0 0 0 1>}
  n66 = NamedRerouteUsage()
  n67 = Saturate(None=Sine(None=Multiply(A=@n66) {constB=0.5}))
  n68 = Power(Base=Multiply(A=OneMinus(None=@n67))) {constExponent=2.5}
  NamedRerouteDeclaration(None=Max(A=SmoothStep(Value=@n67) {constMax=0.15, constMin=0.2}, B=Saturate(None=@n68))) {name=DrumKitMask, nodeColor=<rgba 1 0.3281 0 1>}
  n69 = TextureCoordinate()
  n70 = Add(A=Constant2Vector(), B=@n69)
  n71 = Constant2Vector() {g=2, r=1}
  n72 = TextureSample(
   UVs=StaticSwitchParameter(
    True=MaterialFunctionCall(UVs=@n70, Texture Scale=@n71) {materialFunction=&ScaleUVsByCenter}.UVs
    False=@n69
   ) {parameterName=IsPilgrimInverseTrack}
   Tex=TextureObject() {texture=/PilgrimCore/GemAndTrack/Track/Textures/T_Sparks_Track_Divider}
  )
  n73 = Multiply(A=@n72.B) {constB=2}
  n74 = ScalarParameter() {parameterName=IsPlastic}
  n75 = ScalarParameter() {parameterName=Is5Lane}
  n76 = Constant() {r=0.01}
  n77 = Constant() {r=0.75}
  n78 = ScalarParameter() {defaultValue=1, parameterName=DividerPosition}
  n79 = Saturate(None=@n78)
  n80 = Subtract(A=@n78)
  n81 = LinearInterpolate(Alpha=Saturate(None=@n78)) {constA=0.275, constB=0.09}
  n82 = Saturate(None=Subtract(A=@n78))
  n83 = Subtract(A=@n78) {constB=2}
  n84 = LinearInterpolate(
   A=LinearInterpolate(A=LinearInterpolate(Alpha=@n79) {constA=0.22, constB=-0.001}, Alpha=Saturate(None=@n80)) {constB=-0.22}
   B=LinearInterpolate(A=LinearInterpolate(A=@n81, Alpha=@n82) {constB=-0.091}, Alpha=Saturate(None=@n83)) {constB=-0.273}
   Alpha=ScalarParameter() {parameterName=Is5Lane}
  )
  n85 = StaticBool() {value=true}
  n86 = MaterialFunctionCall(Offset=Constant() {r=-0.1}) {materialFunction=&GeneratedBand}
  n87 = Multiply(A=Multiply(
    A=MaterialFunctionCall(Width=@n76, Sharpness=@n77, Offset=@n84, Direction Switch=@n85) {materialFunction=&GeneratedBand}.Result
    B=Power(Base=@n86.Result)
   )) {constB=5}
  n88 = ScalarParameter() {defaultValue=1, parameterName=UseDivider}
  n89 = Add(
   A=Multiply(A=Multiply(A=@n73, B=@n74), B=@n75)
   B=Add(A=LinearInterpolate(A=@n72.R, B=@n72.R, Alpha=@n75), B=Multiply(A=@n87, B=@n88) {constB=0})
  )
  n90 = Multiply(A=LinearInterpolate(Alpha=NamedRerouteUsage()) {constA=1, constB=2}, B=NamedRerouteUsage())
  n91 = NamedRerouteUsage()
  n92 = LinearInterpolate(Alpha=NamedRerouteUsage()) {constA=5, constB=10}
  n93 = LinearInterpolate(B=Multiply(A=@n89, B=Multiply(A=@n91, B=@n92))) {constAlpha=1}
  n94 = Constant()
  NamedRerouteDeclaration(None=Add(A=Multiply(A=@n89, B=@n90), B=StaticSwitchParameter(True=@n93, False=@n94) {parameterName=IsPilgrimInverseTrack})) {name=DividerLines, nodeColor=<rgba 0 0 0 1>}
  n95 = MaterialFunctionCall() {materialFunction=/Engine/Functions/Engine_MaterialFunctions02/PivotPainter/PivotPainter_PerObjectData}
  NamedRerouteDeclaration(None=@n95.Random Value Per Element) {name=RandomPerPiece, nodeColor=<rgba 0 0 0 1>}
  NamedRerouteDeclaration(None=@n95.Pivot Position) {name=PivotPosition, nodeColor=<rgba 0 0 0 1>}
  NamedRerouteDeclaration(None=@n95.X-Axis Vector) {name=X-AxisVector, nodeColor=<rgba 0 0 0 1>}
  n96 = ScalarParameter() {parameterName=TrackPhaseInOut}
  n97 = Subtract(A=ComponentMask(None=TextureCoordinate()) {g=true}, B=OneMinus(None=Multiply(A=@n96) {constB=1.5}))
  n98 = Constant() {desc=0.5, r=0.5}
  NamedRerouteDeclaration(None=Saturate(None=Divide(A=@n97, B=@n98) {constB=0.5})) {name=AnimMask, nodeColor=<rgba 0 0 0 1>}
  n99 = Multiply(
   A=Subtract(A=NamedRerouteUsage(), B=WorldPosition().XYZ)
   B=Power(Base=NamedRerouteUsage(), Exp=Constant() {desc=0.5, r=0.5}) {constExponent=1}
  )
  n100 = Subtract(A=NamedRerouteUsage(), B=ObjectPositionWS())
  n101 = NamedRerouteUsage()
  n102 = Power(Base=@n101, Exp=Constant() {desc=2, r=1})
  n103 = VectorParameter() {desc="0,0,-20", defaultValue=<rgba 0 0 -20 1>}
  n104 = NamedRerouteUsage()
  n105 = Constant() {r=0.7}
  n106 = Multiply(A=@n103.RGB, B=Multiply(A=Clamp(None=@n104, Min=@n105) {minDefault=0.5}, B=Constant() {desc=5, r=1}) {constB=5})
  n107 = Add(
   A=Multiply(A=Multiply(A=@n100, B=@n102), B=Constant() {desc=0.5, r=-0.25}) {constB=0.5}
   B=Multiply(A=@n101, B=Transform(None=@n106) {transformSourceType=TRANSFORMSOURCE_Local})
  )
  NamedRerouteDeclaration(None=StaticSwitchParameter(True=Add(A=@n99, B=@n107), False=Constant()) {defaultValue=true, parameterName=IsAnimatedTrack}) {name=WPO, nodeColor=<rgba 0 0 0 1>}
  n108 = NamedRerouteUsage()
  n109 = Fresnel(ExponentIn=Constant() {r=5}) {exponent=10}
  n110 = Constant() {desc=10, r=7}
  n111 = Multiply(
   A=Multiply(A=Power(Base=@n108) {constExponent=3}, B=Multiply(A=@n109, B=@n110) {constB=4})
   B=MaterialFunctionCall(Hue Shift Percentage=Constant() {r=0.1}, Texture=NamedRerouteUsage()) {materialFunction=&HueShift}.Result
  )
  NamedRerouteDeclaration(None=StaticSwitchParameter(True=Multiply(A=@n111) {constB=250}, False=Constant()) {defaultValue=true, parameterName=IsAnimatedTrack}) {name=Track Anim Emissive, nodeColor=<rgba 0 0 0 1>}
  n112 = Constant() {r=0.06}
  n113 = Constant() {r=1}
  n114 = Constant() {r=0.5}
  n115 = StaticBool() {value=true}
  n116 = Multiply(
   A=Power(Base=MaterialFunctionCall(Width=@n112, Sharpness=@n113, Offset=@n114, Direction Switch=@n115) {materialFunction=&GeneratedBand}.Result) {constExponent=4}
   B=Constant() {r=0.15}
  ) {constB=5}
  n117 = Constant() {r=0.6}
  n118 = StaticBool() {value=true}
  n119 = Power(Base=OneMinus(None=MaterialFunctionCall(Width=@n117, Direction Switch=@n118) {materialFunction=&GeneratedBand}.Result))
  n120 = NamedRerouteUsage()
  n121 = LinearInterpolate(Alpha=NamedRerouteUsage()) {constA=17, constB=23}
  n122 = Multiply(A=Multiply(A=Add(A=@n116, B=@n119), B=Multiply(A=@n120, B=@n121))) {constB=2}
  n123 = SmoothStep(
   Min=Constant()
   Max=Constant() {r=1}
   Value=MaterialFunctionCall() {materialFunction=&LinearGradient}.VGradient
  )
  n124 = NamedRerouteUsage()
  n125 = SmoothStep(
   Min=Constant() {r=1}
   Max=Constant() {r=0.7}
   Value=MaterialFunctionCall() {materialFunction=&LinearGradient}.VGradient
  )
  n126 = Multiply(A=Multiply(A=Multiply(A=@n122, B=@n123), B=OneMinus(None=@n124)), B=@n125)
  n127 = NamedRerouteUsage()
  n128 = TextureCoordinate()
  n129 = Multiply(A=Power(Base=ComponentMask(None=@n128) {g=true}, Exp=Constant() {r=2.5}) {constExponent=0.8}, B=Constant() {r=1.5}) {constB=1.5}
  n130 = Multiply(A=@n125, B=Multiply(A=NamedRerouteDeclaration(None=@n127) {name=BaseColor, nodeColor=<rgba 0 1 0.2031 1>}, B=Saturate(None=@n129)))
  NamedRerouteDeclaration(None=Add(
    A=NamedRerouteDeclaration(None=@n126) {name=Side Glow, nodeColor=<rgba 0 0 0 1>}
    B=NamedRerouteDeclaration(None=@n130) {name=TrackColor, nodeColor=<rgba 0 0 0 1>}
   )) {name=Track Base, nodeColor=<rgba 0 0 0 1>}
  n131 = Multiply(
   A=MaterialFunctionCall(Repeat Count=Constant2Vector() {g=1, r=2}, texcoord=TextureCoordinate()) {materialFunction=&MF_UVMirror}.Result
   B=Constant2Vector() {g=5, r=1}
  )
  n132 = Constant2Vector() {g=-0.2, r=-0.1}
  NamedRerouteDeclaration(None=TextureSample(UVs=Panner(Coordinate=@n131, Speed=@n132)) {texture=/PilgrimCore/GemAndTrack/Track/Textures/T_Sparks_Track_PeakState}.R) {name=TrianglesTexture, nodeColor=<rgba 0 0 0 1>}
  n133 = Add(A=ComponentMask(None=TextureCoordinate()) {g=true}) {constB=-0.35}
  n134 = Step(Y=@n133) {constX=0.5}
  n135 = LinearInterpolate(A=OneMinus(None=@n133), B=@n133, Alpha=@n134)
  n136 = Multiply(A=@n135, B=@n134)
  n137 = Saturate(None=Multiply(A=SmoothStep(Value=@n136)) {constB=1.25}) {desc=Early gradient}
  n138 = Multiply(A=@n135, B=OneMinus(None=@n134))
  n139 = Saturate(None=Multiply(A=SmoothStep(Value=@n138) {constMin=0.35}) {constB=4}) {desc=Late gradient}
  NamedRerouteDeclaration(None=MaterialFunctionCall(R=Add(A=@n137, B=@n139) {desc=Perfect gradient}, G=@n137, B=@n139) {materialFunction=&AppendMany}.RGB) {name=DrumKitAccuracyGradients, nodeColor=<rgba 0.4531 0 1 1>}
  n140 = Add(A=ComponentMask(None=TextureCoordinate()) {g=true}) {constB=-0.35}
  n141 = OneMinus(None=@n140)
  n142 = Step(Y=@n140) {constX=0.5}
  n143 = Multiply(A=SmoothStep(Value=LinearInterpolate(A=@n141, B=@n140, Alpha=@n142)) {constMin=0.45}) {constB=50}
  NamedRerouteDeclaration(None=LinearInterpolate(A=Saturate(None=@n143), Alpha=NamedRerouteUsage()) {constB=0}) {name=SmashersMask, nodeColor=<rgba 0 1 0.1094 1>}
  NamedRerouteDeclaration(None=If(A=ScalarParameter() {parameterName=Is5Lane}, "A > B"=Constant() {r=5}, "A < B"=Constant() {r=4}) {constB=1}) {name=LaneCount, desc="Number of lanes on the track, either 4 or 5", nodeColor=<rgba 0.7578 1 0 1>}
  n144 = NamedRerouteUsage()
  n145 = NamedRerouteDeclaration(None=Constant() {r=0.05}) {name=LaneMargin, desc=UV margin around lanes, nodeColor=<rgba 0 0.1562 1 1>}
  n146 = OneMinus(None=Multiply(A=@n145) {constB=2}) {desc=Total width of all lanes without margins}
  n147 = Saturate(None=Divide(A=Subtract(A=@n144, B=@n145), B=@n146)) {desc=Normalize so that margins don't get an out-of-bounds index}
  n148 = NamedRerouteUsage()
  NamedRerouteDeclaration(None=Floor(None=Multiply(A=@n147, B=@n148))) {name=LaneIndex, nodeColor=<rgba 0 1 0.4844 1>}
  NamedRerouteDeclaration(None=Divide(A=@n146, B=@n148)) {name=LaneWidth, nodeColor=<rgba 1 0.1641 0 1>}
  n149 = NamedRerouteUsage()
  n150 = Add(A=Multiply(A=NamedRerouteUsage(), B=NamedRerouteUsage()), B=NamedRerouteUsage()) {desc=Lane UV start}
  n151 = NamedRerouteUsage()
  n152 = Sine(None=Divide(A=Subtract(A=@n149, B=@n150), B=Multiply(A=@n151) {constB=2}) {desc="[0-0.5] across width of lane\r\nGives us half a sine wave period starting and ending at 0"}) {desc="Half of a sine wave, "}
  NamedRerouteDeclaration(None=SmoothStep(Value=Power(Base=@n152) {constExponent=4, desc=Exponential falloff}) {constMax=0.1}) {name=LaneAccuracyMask, nodeColor=<rgba 1 0 0.7969 1>}
  NamedRerouteDeclaration(None=Constant3Vector() {constant=<rgba 0.9911 0.6867 0 1>}) {name=PerfectAccuracyColor, nodeColor=<rgba 0.9911 0.6867 0 1>}
  NamedRerouteDeclaration(None=Constant3Vector() {constant=<rgba 0 0.4852 0.9911 1>}) {name=GoodAccuracyColor, nodeColor=<rgba 0 0.4852 0.9911 1>}
  n153 = Multiply(A=@n38.Color_1, B=@n39) {desc=Becomes 0 if not using a drumkit}
  NamedRerouteDeclaration(None=ComponentMask(None=Add(A=@n41.Color_1, B=@n153)) {b=true, g=true, r=true}) {name=AccuracyColor, nodeColor=<rgba 0 0.5781 1 1>}
  NamedRerouteDeclaration(None=LinearInterpolate(
    A=VectorParameter() {parameterName=RailColor, defaultValue=<rgba 0.0082 0.0082 0.02 1>}.RGB
    B=VectorParameter() {parameterName=RailColorPlastic, defaultValue=<rgba 0.01 0.01 0.01 1>}.RGB
    Alpha=ScalarParameter() {parameterName=IsPlastic}
   )) {name=DefaultRailColor, nodeColor=<rgba 1 0.4453 0 1>}
  NamedRerouteUsage()
  n154 = Time()
  n155 = Constant() {r=0.75}
  OneMinus(None=Frac(None=Multiply(A=@n154, B=@n155) {constB=0.5}))
  Reroute(None=@n103.RGB)
  ScalarParameter() {defaultValue=1, parameterName=Max Streak On/Off Test Value}
  ScalarParameter() {parameterName=Overdrive On/Off Test Value}
 ]
}
