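# Material function that pins Pilgrim track meshes or particles to the camera as a screen-space overlay through World Position Offset.
# It re-bases vertices into view space, applies screen offset, rotation and FOV correction, then compresses depth toward the near clip plane.
# Settings come from MPC_PilgrimTrack, with cine-camera blending, or from local parameters.
# It outputs attributes, WPO and adjusted normals, with ShadowReplace zeroing the offset for shadows.
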MaterialFunction /PilgrimCore/GemAndTrack/Track/Materials/MF_PilgrimTrackWPO
graph {nodes [
  Constant3Vector()
  Constant3Vector() {constant=<rgba 100 0 0 0>}
  Constant()
  n1 = FunctionInput() {bUsePreviewValueAsDefault=true, inputType=FunctionInput_MaterialAttributes}
  n2 = MaterialFunctionCall() {materialFunction=&CameraDirectionVector}
  n3 = WorldPosition() {worldPositionShaderOffset=WPT_CameraRelative}
  n4 = Add(A=NamedRerouteUsage(), B=@n3.XYZ)
  n5 = Subtract(A=Multiply(A=@n2.Result, B=DotProduct(A=@n2.Result, B=@n4)), B=@n4)
  n6 = NamedRerouteUsage()
  n7 = OneMinus(None=Divide(
    A=MaterialFunctionCall() {materialFunction=/Engine/Functions/Engine_MaterialFunctions02/Utility/FOV}."FOV "
    B=Tangent!(None=@n6) {period=720}
   ))
  n8 = Add(A=Add(A=Multiply(A=@n5, B=@n7), B=NamedRerouteUsage()), B=@n3.XYZ)
  n9 = Normalize(VectorInput=@n8)
  n10 = MaterialFunctionCall() {materialFunction=&CameraDirectionVector}
  n11 = Multiply(A=Divide(A=@n9, B=DotProduct(A=@n9, B=@n10.Result)), B=Custom!(None) {code=View.NearPlane, description=NearClip, outputType=CMOT_Float1})
  n12 = Subtract(A=@n8, B=@n11)
  MaterialExpressionDotProduct_2 = DotProduct(A=@n12, B=MaterialFunctionCall() {materialFunction=&CameraDirectionVector}.Result)
  n13 = Custom!(X=@MaterialExpressionDotProduct_2) {
   code="return step(X,0.0);"
   description=Plus
   outputType=CMOT_Float1
   inputs [{inputName=X, input {expression=&self:MaterialExpressionDotProduct_2}}]
  }
  n14 = NamedRerouteUsage()
  n15 = Add(A=Multiply(A=@n12, B=Max(A=@n13, B=@n14)), B=@n11)
  n16 = ShadowReplace(Default=Multiply(A=Subtract(A=@n15, B=@n3.XYZ), B=NamedRerouteUsage()), Shadow=Constant())
  n17 = SetMaterialAttributes!(MaterialAttributes=@n1, WorldPositionOffset=@n16) {attributeSetTypes [#1]}
  FunctionOutput(None=@n17)
  FunctionOutput(None=Constant3Vector()) {outputName=Debug, sortPriority=2}
  n18 = StaticBoolParameter() {defaultValue=true, group=PilgrimTrackWPO, parameterName=UsePilgrimTrackWPO}
  FunctionOutput(None=StaticSwitch(True=@n16, False=Constant(), Value=@n18)) {outputName=WPO, sortPriority=1}
  n19 = VertexNormalWS()
  n20 = Normalize(VectorInput=Transform(None=Constant3Vector() {constant=<rgba 1 0 0 0>}) {transformSourceType=TRANSFORMSOURCE_View})
  n21 = Divide(A=NamedRerouteUsage()) {constB=360}
  n22 = NamedRerouteUsage()
  n23 = RotateAboutAxis(NormalizedRotationAxis=@n20, RotationAngle=@n21, PivotPoint=Constant(), Position=@n22) {desc=wo}
  n24 = NamedRerouteDeclaration(None=Normalize(VectorInput=Add(A=@n23, B=@n22))) {name=Normal, nodeColor=<rgba 0.6641 0 1 1>}
  n25 = NamedRerouteUsage()
  FunctionOutput(None=StaticSwitch(
    True=LinearInterpolate(A=@n19, B=@n24, Alpha=@n25)
    False=@n19
    Value=StaticBoolParameter() {defaultValue=true, group=PilgrimTrackWPO, parameterName=UsePilgrimTrackWPO}
   )) {outputName=WSNormal}
  MaterialFunctionCall!(WorldPosition, PivotLocation, WS Normals, Custom Object Basis 1, Custom Object Basis 2, Custom Object Basis 3) {materialFunction=/Engine/Functions/Engine_MaterialFunctions02/WorldPositionOffset/AlignMeshToTheCamera}
  MaterialFunctionCall!(Object Basis 1, Object Basis 2, Object Basis 3, Camera Space Position Offset) {materialFunction=/Engine/Functions/Engine_MaterialFunctions02/WorldPositionOffset/AttachMeshToTheCamera}
  n26 = Subtract(
   A=MaterialFunctionCall() {materialFunction=&ObjectPivotPoint}.Object Pivot Location
   B=ActorPositionWS()
  )
  n27 = StaticSwitch(True=ParticlePositionWS(), False=WorldPosition().XYZ, Value=NamedRerouteUsage())
  n28 = Multiply(
   A=MaterialFunctionCall() {materialFunction=/Engine/Functions/Engine_MaterialFunctions02/WorldPositionOffset/ObjectScale}.Scale XYZ
   B=TransformPosition(None=Add(A=@n26, B=@n27)) {transformSourceType=TRANSFORMPOSSOURCE_World, transformType=TRANSFORMPOSSOURCE_Instance}
  ) {desc=Apply Object Scale once again}
  n29 = MaterialFunctionCall!(
   Vector1=Constant3Vector() {constant=<rgba 1 0 0 0>}
   Vector2=Constant3Vector() {constant=<rgba 0 0 -1 0>}
   Vector3=Constant3Vector() {constant=<rgba 0 1 0 0>}
  ) {materialFunction=/Engine/Functions/Engine_MaterialFunctions02/Math/MakeVectorsOrthogonal}
  n30 = NamedRerouteUsage()
  n31 = Add(
   A=Transform(None=MaterialFunctionCall(VectorToTransform=@n28, BasisX=@n29.Vector1, BasisY=@n29.Vector1, BasisZ=@n29.Vector1, BasisPosition=@n30) {materialFunction=&Transform3x3Matrix}.Transformed Vector) {transformSourceType=TRANSFORMSOURCE_View}
   B=CameraPositionWS()
  )
  n32 = ParticlePositionWS()
  n33 = WorldPosition()
  n34 = NamedRerouteUsage()
  NamedRerouteDeclaration(None=Subtract(A=@n31, B=StaticSwitch(True=@n32, False=@n33.XYZ, Value=@n34))) {name=CameraAttachPreRotation, nodeColor=<rgba 0.9922 1 0 1>}
  n35 = ComponentMask(None=CollectionParameter() {collection=/PilgrimCore/GemAndTrack/MPC_PilgrimTrack, parameterName=PilgrimTrack_ScreenOffset}) {b=true, g=true, r=true}
  n36 = ComponentMask(None=CollectionParameter() {collection=&MPC_PilgrimTrack, parameterName=PilgrimTrack_CineCamOffset}) {b=true, g=true, r=true}
  n37 = CollectionParameter() {collection=&MPC_PilgrimTrack, parameterName=PilgrimTrack_CineCamEnabled}
  NamedRerouteDeclaration(None=StaticSwitch(
    True=LinearInterpolate(A=@n35, B=@n36, Alpha=@n37)
    False=VectorParameter() {group=PilgrimTrackWPO, parameterName=PTW_ScreenOffset, defaultValue=<rgba 22 -22 200 1>}.RGB
    Value=NamedRerouteUsage()
   )) {name=ScreenOffset, nodeColor=<rgba 1 0 0.3047 1>}
  n38 = Transform(None=VertexNormalWS()) {transformSourceType=TRANSFORMSOURCE_World, transformType=TRANSFORM_Instance}
  n39 = MaterialFunctionCall(
   Vector1=Constant3Vector() {constant=<rgba 1 0 0 0>}
   Vector2=Constant3Vector() {constant=<rgba 0 0 -1 0>}
   Vector3=Constant3Vector() {constant=<rgba 0 1 0 0>}
  ) {materialFunction=&MakeVectorsOrthogonal}
  n40 = Constant3Vector()
  NamedRerouteDeclaration(None=Transform(None=MaterialFunctionCall(VectorToTransform=@n38, BasisX=@n39.Vector1, BasisY=@n39.Vector1, BasisZ=@n39.Vector1, BasisPosition=@n40) {materialFunction=&Transform3x3Matrix}.Transformed Vector) {transformSourceType=TRANSFORMSOURCE_View}) {name=TransformedNormal, nodeColor=<rgba 1 0.1641 0 1>}
  n41 = Add(A=Transform(None=NamedRerouteUsage()) {transformSourceType=TRANSFORMSOURCE_View}, B=CameraPositionWS())
  n42 = NamedRerouteUsage()
  NamedRerouteDeclaration(None=Add(A=RotateAboutAxis(NormalizedRotationAxis=@n20, RotationAngle=@n21, PivotPoint=@n41, Position=@n42), B=NamedRerouteUsage())) {name=AttachToCam, nodeColor=<rgba 1 0.7031 0 1>}
  NamedRerouteDeclaration(None=@n31) {name=CameraAttachWSP, nodeColor=<rgba 0 1 0.25 1>}
  n43 = CollectionParameter() {collection=&MPC_PilgrimTrack, parameterName=PilgrimTrack_RotationAngle}
  n44 = CollectionParameter() {collection=&MPC_PilgrimTrack, parameterName=PilgrimTrack_CineCamRotationAngle}
  n45 = CollectionParameter() {collection=&MPC_PilgrimTrack, parameterName=PilgrimTrack_CineCamEnabled}
  NamedRerouteDeclaration(None=StaticSwitch(
    True=LinearInterpolate(A=@n43, B=@n44, Alpha=@n45)
    False=ScalarParameter() {defaultValue=-20, group=PilgrimTrackWPO, parameterName=PTW_RotationAngle}
    Value=NamedRerouteUsage()
   )) {name=RotationAngle, nodeColor=<rgba 0.6406 0 1 1>}
  NamedRerouteDeclaration(None=StaticSwitch(
    True=CollectionParameter() {collection=&MPC_PilgrimTrack, parameterName=PilgrimTrack_FOV}
    False=ScalarParameter() {defaultValue=47, group=PilgrimTrackWPO, parameterName=PTW_FOV}
    Value=NamedRerouteUsage()
   )) {name=FOV, nodeColor=<rgba 0.3125 0 1 1>}
  NamedRerouteDeclaration(None=StaticSwitch(
    True=CollectionParameter() {collection=&MPC_PilgrimTrack, parameterName=PilgrimTrack_ScaleInDepth}
    False=ScalarParameter() {defaultValue=0.01, group=PilgrimTrackWPO, parameterName=PTW_ScaleInDepth}
    Value=NamedRerouteUsage()
   )) {name=ScaleInDepth, nodeColor=<rgba 0.875 0 1 1>}
  NamedRerouteDeclaration(None=StaticSwitch(
    True=CollectionParameter() {collection=&MPC_PilgrimTrack, parameterName=PilgrimTrack_WPOEnabled}
    False=ScalarParameter() {defaultValue=1, group=PilgrimTrackWPO, parameterName=PTW_WPOEnabled}
    Value=NamedRerouteUsage()
   )) {name=WPOEnabled, nodeColor=<rgba 1 0.8203 0 1>}
  NamedRerouteDeclaration(None=StaticBoolParameter() {defaultValue=true, group=PilgrimTrackWPO, parameterName=PTW_UseMPCValues}) {name=UseMPC, nodeColor=<rgba 1 0.6797 0 1>}
  NamedRerouteDeclaration(None=FunctionInput(Preview=StaticBool()) {bUsePreviewValueAsDefault=true, inputName=IsParticle, inputType=FunctionInput_StaticBool}) {name=IsParticle, nodeColor=<rgba 0.6406 1 0 1>}
  Normalize(VectorInput=@n22)
  ParticleSize()
  StaticSwitch(True=@n17, False=@n1, Value=@n18)
 ]}
