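# M_Track_Gutter is a translucent, unlit material for the gutter of the gem track.
# It calls MF_PilgrimTrackWPO, the track's shared function, presumably for world position offset.
# A gradient comes from V-tiled UVs through a Panner with constant time, giving a fixed shift, then a green-channel mask, inversion and power falloff.
# The graph shows no links from these nodes to material outputs.
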Material /PilgrimCore/GemAndTrack/Track/Materials/M_Track_Gutter
graph {
 properties {blendMode=BLEND_Translucent, pixelDepthOffsetMode=PDOM_Legacy, shadingModel=MSM_Unlit, usedShadingModels=Unlit}
 nodes [
  MaterialFunctionCall!(IsParticle, In) {materialFunction=&/MF_PilgrimTrackWPO}
  n1 = Panner(Coordinate=TextureCoordinate() {vTiling=1.1}, Time=Constant() {r=-0.3}) {speedX=1, speedY=1}
  Power(Base=OneMinus(None=ComponentMask(None=@n1) {g=true})) {constExponent=2.5}
 ]
}
