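# Material function that builds a mask for a single track lane, named for per-lane feedback in the GemAndTrack track materials.
# The lane position blends between 4-lane and 5-lane X inputs by Is5Lane and is subtracted from UV X.
# A sine of that offset is raised to the LaneWidthOffset exponent to narrow the band, then SmoothStep with LaneMax as upper edge outputs LaneMask.
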MaterialFunction /PilgrimCore/GemAndTrack/Track/Materials/M_Track_PerLaneFeedbackLaneMask
graph {nodes [
  n1 = FunctionInput() {inputName=4LaneX, inputType=FunctionInput_Scalar, sortPriority=20}
  n2 = FunctionInput() {inputName=5LaneX, inputType=FunctionInput_Scalar, sortPriority=30}
  n3 = FunctionInput() {inputName=Is5Lane, inputType=FunctionInput_Scalar, sortPriority=10}
  n4 = Sine(None=Subtract(A=FunctionInput() {inputName=UV_X, inputType=FunctionInput_Scalar}, B=LinearInterpolate(A=@n1, B=@n2, Alpha=@n3)) {constB=-0.11})
  n5 = FunctionInput() {inputName=LaneWidthOffset, inputType=FunctionInput_Scalar, sortPriority=40}
  FunctionOutput(None=SmoothStep(Max=FunctionInput() {inputName=LaneMax, inputType=FunctionInput_Scalar, sortPriority=50}, Value=Power(Base=@n4, Exp=@n5) {constExponent=85}) {constMax=0.1}) {outputName=LaneMask}
 ]}
